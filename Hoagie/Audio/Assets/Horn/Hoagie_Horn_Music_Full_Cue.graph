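# Sound cue that plays the full Hoagie horn music track as a distance-attenuated sound.
# A single wave player feeds the output directly, with no randomization or mixing nodes.
# It uses distant horn attenuation with a long audible range, a boosted volume, the POI music sound class and a horn music concurrency set that governs overlapping plays.
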SoundCue /Hoagie/Audio/Assets/Horn/Hoagie_Horn_Music_Full_Cue
root=WavePlayer_0
nodes [{id=WavePlayer_0, properties {soundWaveAssetPtr=&/Hoagie_Horn_Music_Full_01}}]
properties {
 attenuationSettings=/Hoagie/Audio/Hoagie_Horn_Distant_Att
 maxDistance=23500
 soundClassObject=/Game/Sounds/FortSoundClassesAndModes/Fort_Music_POI_PSM
 volumeMultiplier=1.7
 concurrencySet [&/Hoagie_Horn_Music_Conc]
}
